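# Niagara system for the BriskReign Glacier character cosmetic that emits a short burst of dark blue-grey smoke sprites from six sockets.
# One emitter partitions the burst across socket transforms gathered at system level, easing particles toward randomized socket-relative offsets while blending from their current position.
# Sprites spin, grow, animate through randomized SubUV frames and fade, with opacity adjusted for time of day and front end.
# User head-hidden and hand-hidden flags kill particles by index modulo six.
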NiagaraSystem /BRCosmetics/Characters/Player/Male/Medium/Bodies/M_MED_BriskReign_Glacier/FX/NS_Character_BriskReign_Glacier_Impulse
dependencies {
 usedDataInterfaces [
  &NiagaraDataInterfaceCurve
  &NiagaraDataInterfaceVector2DCurve
  &NiagaraDataInterfaceSpriteRendererInfo
  &NiagaraDataInterfaceColorCurve
  &NiagaraDataInterfaceSkeletalMesh
 ]
 usedDynamicInputs [&Vector2DFromCurve, &ColorFromCurve]
 usedModules [
  &SystemState
  &self:SystemSpawnScript.SystemScriptSource.SystemScriptGraph.NiagaraNodeAssignment_2.SetVariables_#1
  &self:SystemSpawnScript.SystemScriptSource.SystemScriptGraph.NiagaraNodeAssignment_0.SetVariables_#2
  &self:GetMultiSockets
 ]
}
emitters {SmokeBurst {
  emitterUpdateScript [
   &EmitterState(Inactive Response=::NewEnumerator1, Life Cycle Mode=::NewEnumerator1, Loop Behavior=::NewEnumerator1, Loop Duration=2, UseLoopDelay=false)
   &SpawnBurst_Instantaneous(Spawn Count=18, Use Spawn Probability=false, UseLoopCountLimit=false)
  ]
  particleSpawnScript [
   &PartitionParticles(Number Of Partitions=6)
   &InitializeParticle(
    Color=<rgba 0.0154 0.024 0.04 1.5>
    Lifetime Max=1.2
    Lifetime Min=0.5
    Lifetime Mode=::NewEnumerator1
    Sprite Rotation Mode=::NewEnumerator2
    Sprite Size=&Multiply_Vector2D_ByFloat(Float=1, Vector2D=&RandomRangeVector2D(Minimum=<50 50>, Maximum=<60 60>))
    Sprite Size Mode=::NewEnumerator4
    UsePositionOffset=true
    Write Material Random=true
   )
   &CameraOffset(Camera Offset Amount=10)
   /SetVariables(
    Particles.CustomID:NiagaraInt32=$Particles.PartitionParticles.Partition
    Particles.CustomRandom:NiagaraInt32=&RandomRangeInt(Minimum=0, Maximum=5)
   ) {isSetParameters=true}
  ]
  particleUpdateScript [
   &ParticleState()
   /SetVariables(Particles.Position=&Lerp_Vector(
     A=&AddVectorToPosition(
      Position=&MergeHelper_VectorSwitch(
       0=$System.Socket Translation
       1=$System.Socket Translation001
       2=$System.Socket Translation002
       3=$System.Socket Translation003
       4=$System.Socket Translation004
       5=$System.Socket Translation005
       Index=$Particles.CustomID
       Num=::NewEnumerator4
      )
      Vector=&Lerp_Vector(
       A=<0 0 0>
       B=&Add_Vector(
        A=&RotateVectorByQuaternion(
         Vector To Rotate=&RandomRangeVector(Minimum=<-5 -2 -2>, Maximum=<-10 2 2>)
         Quaternion=/Game/Effects/Niagara/Modules/MergeEmitters/MergeHelper_QuatSwitch(
          0:Quat4f=$System.Socket Rotation
          1:Quat4f=$System.Socket Rotation001
          2:Quat4f=$System.Socket Rotation002
          3:Quat4f=$System.Socket Rotation003
          4:Quat4f=$System.Socket Rotation004
          5:Quat4f=$System.Socket Rotation005
          6:Quat4f=?
          7:Quat4f=?
          8:Quat4f=?
          9:Quat4f=?
          Index:NiagaraInt32=$Particles.CustomID
          Num:MergeHelper_SwitcherEnum=::NewEnumerator4
         )
        )
        B=<0 0 0>
       )
       Alpha=&FloatFromCurve(FloatCurve=DataInterface() {propertyValues {Curve {keys [
            {arriveTangent=1.6379, interpMode=RCIM_Cubic, leaveTangent=1.6379, tangentMode=RCTM_User, time=0, value=0}
            {interpMode=RCIM_Cubic, leaveTangent=1.3767, time=1, value=1}
           ]}}})
      )
     )
     B=Linked() {linkedVariable {name=Particles.Position, type=NiagaraPosition}}
     Alpha=&FloatFromCurve(
      FloatCurve=DataInterface() {propertyValues {Curve {defaultValue=1, keys [<at 0 0 RCIM_Cubic>,<at 1 1 RCIM_Cubic>]}}}
      Scale Curve=&RandomRangeFloat(Minimum=0.3, Maximum=0.5)
     )
    )) {isSetParameters=true}
   &SpriteRotationRate(Rotation Rate=&RandomRangeFloat(Minimum=-50, Maximum=50))
   &ScaleSpriteSize(
    Scale Sprite Size Mode=::NewEnumerator0
    Uniform Scale Factor=&FloatFromCurve(FloatCurve=DataInterface() {propertyValues {Curve {
        defaultValue=1
        keys [
         {arriveTangent=2.6678, interpMode=RCIM_Cubic, leaveTangent=2.6678, tangentMode=RCTM_User, time=0, value=0.3}
         {arriveTangent=0.8, interpMode=RCIM_Cubic, leaveTangent=0.8, time=0.2021, value=0.8}
         {arriveTangent=0.0224, interpMode=RCIM_Cubic, leaveTangent=0.0224, tangentMode=RCTM_User, time=1, value=1.1}
        ]
       }}})
   )
   &SubUVAnimation(
    End Frame Range Override:NiagaraInt32=&Add_Integer(A=&RandomRangeInt(Minimum=13, Maximum=15), B=$Particles.Initial.CustomRandom)
    "Sprite Renderer "=DataInterface() {propertyValues {SpriteRenderer=&self:SmokeBurst_0.NiagaraSpriteRendererProperties_0}}
    Start Frame Range Override=$Particles.Initial.CustomRandom
    UseEndFrame=true
    UseStartFrame=true
   )
   &ScaleColor(
    Scale Alpha=&FloatFromCurve(
     FloatCurve=DataInterface() {propertyValues {Curve {keys [{arriveTangent=-1.3408, interpMode=RCIM_Cubic, time=0, value=0.8}, <at 1 0 RCIM_Cubic>]}}}
     Scale Curve=&Multiply_Float(
      A=&FN_FrontEndFloatModifier(Gameplay=&FNTimeOfDayScalar(Night Scale=0.6, Morning Scale=0.9, Evening Scale=0.8), Front End=1)
      B=1
     )
    )
    ScaleA=true
    ScaleRGB=true
   )
   &KillParticles(Kill Particles=&Boolean_OrOperation(
     Bool1=&Boolean_AndOperation(
      Bool1=$User.bIsHeadHidden
      Bool2=&SetBoolByIntComparison(A=&ModuloInt(A=&ReturnExecIndex(), B=6), B=1, Comparison Type=::NewEnumerator0)
     )
     Bool2=&Boolean_AndOperation(
      Bool1=$User.bIsHandHidden
      Bool2=&SetBoolByIntComparison(A=&ModuloInt(A=&ReturnExecIndex(), B=6), B=2, Comparison Type=::NewEnumerator6)
     )
    ))
  ]
  rendererClasses [&NiagaraSpriteRendererProperties]
  renderers [{rendererClass=&NiagaraSpriteRendererProperties}]
 }}
summary {
 systemName=NS_Character_BriskReign_Glacier_Impulse
 emitters [{emitterName=SmokeBurst, rendererClasses [&NiagaraSpriteRendererProperties]}]
 userVariables [
  User.bIsFrontEnd {type=NiagaraBool}
  User.bIsFrontEndPreview {type=NiagaraBool}
  User.bIsHandHidden {type=NiagaraBool}
  User.bIsHeadHidden {type=NiagaraBool}
 ]
}
