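# Niagara cosmetic effect for the BriskReign Glacier character, attaching dark spiral fire and smoke to six skeletal mesh sockets.
# SpiralFire_Arm pins persistent mesh and sprite particles to sockets, passing smoothed positions to materials; SmokeDetailed continuously emits drifting, time-of-day-dimmed flipbook sprites.
# Both read socket transforms from GetMultiSockets by partition ID and scale alpha by SystemFade.
# Head/hand-hidden user flags ramp per-partition visibility instead of killing particles (KillParticles is disabled).
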NiagaraSystem /BRCosmetics/Characters/Player/Male/Medium/Bodies/M_MED_BriskReign_Glacier/FX/NS_Character_BriskReign_Glacier_Reactive
dependencies {
 usedDataInterfaces [
  &NiagaraDataInterfaceCurve
  &NiagaraDataInterfaceVector2DCurve
  &NiagaraDataInterfaceSpriteRendererInfo
  &NiagaraDataInterfaceColorCurve
  &NiagaraDataInterfaceSkeletalMesh
 ]
 usedDynamicInputs [
  &MakeFloatFromVector
  &Vector2DFromCurve
  &RandomRangeVector2D
  &ColorFromCurve
 ]
 usedModules [
  &SystemState
  &self:SystemSpawnScript.SystemScriptSource.SystemScriptGraph.NiagaraNodeAssignment_3.SetVariables_#1
  &self:SystemSpawnScript.SystemScriptSource.SystemScriptGraph.NiagaraNodeAssignment_2.SetVariables_#2
  &self:SystemSpawnScript.SystemScriptSource.SystemScriptGraph.NiagaraNodeAssignment_0.SetVariables_#3
  &self:GetMultiSockets
 ]
}
emitters {
 SmokeDetailed {
  emitterUpdateScript [
   &EmitterState(Inactive Response=::NewEnumerator1, Life Cycle Mode=::NewEnumerator1, Loop Duration=2, UseLoopDelay=false)
   &SpawnRate(SpawnRate=12, Use Spawn Probability=false)
  ]
  particleSpawnScript [
   &PartitionParticles(Number Of Partitions=6)
   &InitializeParticle(
    Color=<rgba 0.0156 0.0241 0.04 1.5>
    Lifetime Max=1.4
    Lifetime Min=1.2
    Lifetime Mode=::NewEnumerator1
    Sprite Rotation Mode=::NewEnumerator2
    Sprite Size Max=<22 22>
    Sprite Size Min=<14 14>
    Sprite Size Mode=::NewEnumerator2
    UsePositionOffset=true
    Write Material Random=true
   )
   &CameraOffset(Camera Offset Amount=10)
   /SetVariables(
    Particles.CustomID:NiagaraInt32=$Particles.PartitionParticles.Partition
    Particles.CustomRandom:NiagaraInt32=&RandomRangeInt(Minimum=0, Maximum=5)
   ) {isSetParameters=true}
   /SetVariables(Particles.Visibility:NiagaraFloat=&MakeCustomFloatFromBool(
     Bool=&Boolean_OrOperation(
      Bool1=&Boolean_AndOperation(
       Bool1=$User.bIsHeadHidden
       Bool2=&SetBoolByIntComparison(A=&ModuloInt(A=$Particles.PartitionParticles.Partition, B=6), B=1, Comparison Type=::NewEnumerator0)
      )
      Bool2=&Boolean_AndOperation(
       Bool1=$User.bIsHandHidden
       Bool2=&SetBoolByIntComparison(A=&ModuloInt(A=$Particles.PartitionParticles.Partition, B=6), B=2, Comparison Type=::NewEnumerator6)
      )
     )
     True Float=0
     False Float=1
    )) {isSetParameters=true}
  ]
  particleUpdateScript [
   &ParticleState()
   /SetVariables(Particles.Visibility:NiagaraFloat=&ClampFloat(Float=&Add_Float(
      A=&Multiply_Float(
       A=&MakeCustomFloatFromBool(
        Bool=&Boolean_OrOperation(
         Bool1=&Boolean_AndOperation(
          Bool1=$User.bIsHeadHidden
          Bool2=&SetBoolByIntComparison(A=&ModuloInt(A=$Particles.PartitionParticles.Partition, B=6), B=1, Comparison Type=::NewEnumerator0)
         )
         Bool2=&Boolean_AndOperation(
          Bool1=$User.bIsHandHidden
          Bool2=&SetBoolByIntComparison(A=&ModuloInt(A=$Particles.PartitionParticles.Partition, B=6), B=2, Comparison Type=::NewEnumerator6)
         )
        )
        True Float=-1
        False Float=1
       )
       B=0.02
      )
      B=$Particles.Visibility
     ))) {isSetParameters=true}
   /SetVariables(Particles.VisibilityScale:NiagaraFloat=&MakeCustomFloatFromBool(
     Bool=&SetBoolByFloatComparison(A=$Particles.Visibility, B=0.1, Comparison Type=::NewEnumerator6)
     True Float=0
     False Float=1
    )) {isSetParameters=true}
   /SetVariables(Particles.Position=&Lerp_Vector(
     A=&AddVectorToPosition(
      Position=&MergeHelper_VectorSwitch(
       0=$System.Socket Translation
       1=$System.Socket Translation001
       2=$System.Socket Translation002
       3=$System.Socket Translation003
       4=$System.Socket Translation004
       5=$System.Socket Translation005
       Index=$Particles.CustomID
       Num=::NewEnumerator4
      )
      Vector=&Lerp_Vector(
       A=<0 0 0>
       B=&Add_Vector(
        A=&RotateVectorByQuaternion(
         Vector To Rotate=&RandomRangeVector(Minimum=<-10 -3 -3>, Maximum=<-20 3 3>)
         Quaternion=/Game/Effects/Niagara/Modules/MergeEmitters/MergeHelper_QuatSwitch(
          0:Quat4f=$System.Socket Rotation
          1:Quat4f=$System.Socket Rotation001
          2:Quat4f=$System.Socket Rotation002
          3:Quat4f=$System.Socket Rotation003
          4:Quat4f=$System.Socket Rotation004
          5:Quat4f=$System.Socket Rotation005
          6:Quat4f=?
          7:Quat4f=?
          8:Quat4f=?
          9:Quat4f=?
          Index:NiagaraInt32=$Particles.CustomID
          Num:MergeHelper_SwitcherEnum=::NewEnumerator4
         )
        )
        B=<0 0 0>
       )
       Alpha=&FloatFromCurve(FloatCurve=DataInterface() {propertyValues {Curve {keys [<at 0 0>,<at 1 1>]}}})
      )
     )
     B=Linked() {linkedVariable {name=Particles.Position, type=NiagaraPosition}}
     Alpha=&FloatFromCurve(
      FloatCurve=DataInterface() {propertyValues {Curve {defaultValue=1, keys [<at 0 0 RCIM_Cubic>,<at 1 1 RCIM_Cubic>]}}}
      Scale Curve=&RandomRangeFloat(Minimum=0.3, Maximum=0.5)
     )
    )) {isSetParameters=true}
   &SpriteRotationRate(Rotation Rate=&RandomRangeFloat(Minimum=-50, Maximum=50))
   &ScaleSpriteSize(
    Scale Sprite Size Mode=::NewEnumerator0
    Uniform Scale Factor=&FloatFromCurve(
     FloatCurve=DataInterface() {propertyValues {Curve {
        defaultValue=1
        keys [
         {arriveTangent=4.1785, interpMode=RCIM_Cubic, leaveTangent=4.1785, tangentMode=RCTM_User, time=0, value=1}
         {arriveTangent=0.6516, interpMode=RCIM_Cubic, leaveTangent=0.6516, tangentMode=RCTM_User, time=0.2042, value=1.7491}
         {arriveTangent=0.0224, interpMode=RCIM_Cubic, leaveTangent=0.0224, tangentMode=RCTM_User, time=1, value=2}
        ]
       }}}
     Scale Curve=$Particles.VisibilityScale
    )
   )
   &SubUVAnimation(
    End Frame Range Override:NiagaraInt32=&Add_Integer(A=&RandomRangeInt(Minimum=10, Maximum=12), B=$Particles.Initial.CustomRandom)
    "Sprite Renderer "=DataInterface() {propertyValues {SpriteRenderer=&self:SmokeDetailed_0.NiagaraSpriteRendererProperties_0}}
    Start Frame Range Override=$Particles.Initial.CustomRandom
    UseEndFrame=true
    UseStartFrame=true
   )
   &ScaleColor(
    Scale Alpha=&FloatFromCurve(
     FloatCurve=DataInterface() {propertyValues {Curve {keys [
         {arriveTangent=3.7485, interpMode=RCIM_Cubic, leaveTangent=3.7485, tangentMode=RCTM_User, time=0, value=0.35}
         {interpMode=RCIM_Cubic, tangentMode=RCTM_User, time=0.5, value=1}
         {tangentMode=RCTM_User, time=1, value=0}
        ]}}}
     Scale Curve=&Multiply_Float(
      A=&FN_FrontEndFloatModifier(Gameplay=&FNTimeOfDayScalar(Night Scale=0.6, Morning Scale=0.9, Evening Scale=0.8), Front End=1)
      B=$System.SystemFade
     )
    )
    ScaleA=true
    ScaleRGB=true
   )
   &KillParticles(Kill Particles=&Boolean_OrOperation(
     Bool1=&Boolean_AndOperation(
      Bool1=$User.bIsHeadHidden
      Bool2=&SetBoolByIntComparison(A=&ModuloInt(A=$Particles.PartitionParticles.Partition, B=6), B=1, Comparison Type=::NewEnumerator0)
     )
     Bool2=&Boolean_AndOperation(
      Bool1=$User.bIsHandHidden
      Bool2=&SetBoolByIntComparison(A=&ModuloInt(A=$Particles.PartitionParticles.Partition, B=6), B=2, Comparison Type=::NewEnumerator6)
     )
    )) {enabled=false}
  ]
  rendererClasses [&NiagaraSpriteRendererProperties]
  renderers [{rendererClass=&NiagaraSpriteRendererProperties}]
 }
 SpiralFire_Arm {
  emitterUpdateScript [
   &EmitterState(
    Inactive Response=::NewEnumerator1
    Life Cycle Mode=::NewEnumerator1
    Loop Behavior=::NewEnumerator1
    Loop Duration Mode=::NewEnumerator1
    UseLoopDelay=false
   )
   &SpawnBurst_Instantaneous(Spawn Count=6, Use Spawn Probability=false, UseLoopCountLimit=false)
  ]
  particleSpawnScript [
   &PartitionParticles(Number Of Partitions=6)
   &InitializeParticle(
    Color=<rgba 0 0 0 3>
    Mesh Scale=&Multiply_Vector(
     A=<1 0.7 0.7>
     B=&MergeHelper_VectorSwitch(
      0=<1 1 1>
      1=<1 -1 1>
      2=<1 1 1>
      3=<1 -1 1>
      4=<1 1 1>
      5=<1 -1 1>
      Index=&ReturnExecIndex()
      Num=::NewEnumerator4
     )
    )
    Mesh Scale Mode=::NewEnumerator4
    UsePositionOffset=true
    Write Material Random=true
   )
   &CameraOffset(Camera Offset Amount=20)
   /SetVariables(Particles.VisibilityTag=$Emitter.SpawnGroup) {isSetParameters=true}
   /SetVariables(Particles.CustomID:NiagaraInt32=$Particles.PartitionParticles.Partition) {isSetParameters=true}
   /SetVariables() {isSetParameters=true}
   /SetVariables(Particles.Visibility:NiagaraFloat=&MakeCustomFloatFromBool(
     Bool=&Boolean_OrOperation(
      Bool1=&Boolean_AndOperation(
       Bool1=$User.bIsHeadHidden
       Bool2=&SetBoolByIntComparison(A=&ModuloInt(A=&ReturnExecIndex(), B=6), B=1, Comparison Type=::NewEnumerator0)
      )
      Bool2=&Boolean_AndOperation(
       Bool1=$User.bIsHandHidden
       Bool2=&SetBoolByIntComparison(A=&ModuloInt(A=&ReturnExecIndex(), B=6), B=2, Comparison Type=::NewEnumerator6)
      )
     )
     True Float=0
     False Float=1
    )) {isSetParameters=true}
  ]
  particleUpdateScript [
   &ParticleState(Kill Particles When Lifetime Has Elapsed=false)
   /SetVariables(Particles.Visibility:NiagaraFloat=&ClampFloat(Float=&Add_Float(
      A=&Multiply_Float(
       A=&MakeCustomFloatFromBool(
        Bool=&Boolean_OrOperation(
         Bool1=&Boolean_AndOperation(
          Bool1=$User.bIsHeadHidden
          Bool2=&SetBoolByIntComparison(A=&ModuloInt(A=&ReturnExecIndex(), B=6), B=1, Comparison Type=::NewEnumerator0)
         )
         Bool2=&Boolean_AndOperation(
          Bool1=$User.bIsHandHidden
          Bool2=&SetBoolByIntComparison(A=&ModuloInt(A=&ReturnExecIndex(), B=6), B=2, Comparison Type=::NewEnumerator6)
         )
        )
        True Float=-1
        False Float=1
       )
       B=0.02
      )
      B=$Particles.Visibility
     ))) {isSetParameters=true}
   /SetVariables(Particles.VisibilityScale:NiagaraFloat=&MakeCustomFloatFromBool(
     Bool=&SetBoolByFloatComparison(A=$Particles.Visibility, B=0.1, Comparison Type=::NewEnumerator6)
     True Float=0
     False Float=1
    )) {isSetParameters=true}
   /SetVariables(Particles.Position=&MergeHelper_VectorSwitch(
     0=$System.Socket Translation
     1=$System.Socket Translation001
     2=$System.Socket Translation002
     3=$System.Socket Translation003
     4=$System.Socket Translation004
     5=$System.Socket Translation005
     Index=$Particles.CustomID
     Num=::NewEnumerator4
    )) {isSetParameters=true}
   /SetVariables(Particles.MeshOrientation=&MergeHelper_QuatSwitch(
     0:Quat4f=$System.Socket Rotation
     1:Quat4f=$System.Socket Rotation001
     2:Quat4f=$System.Socket Rotation002
     3:Quat4f=$System.Socket Rotation003
     4:Quat4f=$System.Socket Rotation004
     5:Quat4f=$System.Socket Rotation005
     6:Quat4f=?
     7:Quat4f=?
     8:Quat4f=?
     9:Quat4f=?
     Index:NiagaraInt32=$Particles.CustomID
     Num:MergeHelper_SwitcherEnum=::NewEnumerator4
    )) {isSetParameters=true}
   &SpriteFacingAndAlignment(
    Sprite Alignment=&RotateVectorByQuaternion(Vector To Rotate=<1 0 0>, Quaternion=$Particles.MeshOrientation)
    WriteAlignment=true
    WriteFacing=true
   )
   &ScaleMeshSize(Scale Factor=&MakeVector(
     X=&FloatFromCurve(
      FloatCurve=DataInterface() {propertyValues {Curve {keys [<at 0 0.2 RCIM_Cubic>,<at 0.6 1 RCIM_Cubic>]}}}
      CurveIndex=$Emitter.Age
      Scale Curve=$Particles.VisibilityScale
     )
     Y=1
     Z=1
    ))
   &ScaleColor(
    Scale Alpha=&Multiply_Float(A=&Multiply_Float(A=$Particles.Visibility, B=0.57), B=$System.SystemFade)
    Scale RGB=&Multiply_VectorByFloat(
     Vector=&MakeVectorFromLinearColorRGB(Color=&RandomRangeLinearColor(Minimum=<rgba 1 1 1 1>, Maximum=<rgba 1.5 1.5 1.5 1>))
     Float=1
    )
    ScaleA=true
    ScaleRGB=true
   )
   &InterpolateOverTime(Rate Of Change=5, Target Vector Value:NiagaraPosition=$Particles.Position, Variable Type=::NewEnumerator2)
   &DynamicMaterialParameters(
    Index 0 Param 1=&RandomRangeFloat(Minimum=1, Maximum=2)
    Index 0 Param 2=3
    Index 0 Param 3=2
    Index 0 Param 4=&RandomRangeFloat(Minimum=1, Maximum=1.6)
    Index 1 Param 4=0.5
    Index 1 Vector 1:Vector3f=$Particles.InterpolateOverTime.InterpolateOverTime_V3_.Moving Average
    Index 2 Param 1=1
    Index 2 Param 2=1
    Index 2 Param 3=1
    Index 2 Param 4=1
    Module.Index0Float0Write=true
    Module.Index0Float1Write=true
    Module.Index0Float2Write=true
    Module.Index0Float3Write=true
    Module.Index1Float3Write=true
    Module.Index1VectorWrite=true
    Module.Index2Float0Write=true
    Module.Index2Float1Write=false
    Module.Index2Float2Write=false
    Module.Index2Float3Write=false
    Write Parameter Index 1=::NewEnumerator3
    Write Parameter Index 2=::NewEnumerator1
   )
   /SetVariables(
    Particles.DynamicParametersAlt:Vector4f=&MakeVector4(X=1, Y=20, Z=1, W=1)
    Particles.DynamicParametersAlt2:Vector4f=&MakeVector4(X=0, Y=0, Z=0, W=1)
   ) {isSetParameters=true}
   &FN_Sprite_Attributes(
    Color=&Multiply_LinearColorByFloat(
     Linear Color=<rgba 0 0 0 0.8>
     Float=&Multiply_Float(A=$System.SystemFade, B=&Multiply_Float(A=0.7, B=$Particles.Visibility))
    )
    Create Color Attribute=true
    Create Position Attribute=true
    Create SpriteSize Attribute=true
    Position:NiagaraPosition=&AddVectorToPosition(
     Position=$Particles.Position
     Vector=&RotateVectorByQuaternion(Vector To Rotate=<10 0 0>, Quaternion=$Particles.MeshOrientation)
    )
    SpriteSize=&Multiply_Vector2D_ByFloat(Float=$Particles.VisibilityScale, Vector2D=&Lerp_Vector2(A=<10 20>, B=<20 40>, Alpha=$System.SystemFade))
    Write to Module Namespace=true
   )
   &KillParticles(Kill Particles=&Boolean_OrOperation(
     Bool1=&Boolean_AndOperation(
      Bool1=$User.bIsHeadHidden
      Bool2=&SetBoolByIntComparison(A=&ModuloInt(A=&ReturnExecIndex(), B=6), B=1, Comparison Type=::NewEnumerator0)
     )
     Bool2=&Boolean_AndOperation(
      Bool1=$User.bIsHandHidden
      Bool2=&SetBoolByIntComparison(A=&ModuloInt(A=&ReturnExecIndex(), B=6), B=2, Comparison Type=::NewEnumerator6)
     )
    )) {enabled=false}
  ]
  rendererClasses [&NiagaraMeshRendererProperties, &NiagaraSpriteRendererProperties]
  renderers [
   {rendererClass=&NiagaraMeshRendererProperties, rendererIndex=0}
   {rendererClass=&NiagaraSpriteRendererProperties, rendererIndex=1}
  ]
 }
}
summary {
 systemName=NS_Character_BriskReign_Glacier_Reactive
 emitters [
  {emitterName=SpiralFire_Arm, rendererClasses [&NiagaraMeshRendererProperties, &NiagaraSpriteRendererProperties]}
  {emitterName=SmokeDetailed, rendererClasses [&NiagaraSpriteRendererProperties]}
 ]
 userVariables [
  User.Alpha {type=NiagaraFloat, defaultValue=1, description=""}
  User.bIsFrontEnd {type=NiagaraBool}
  User.bIsFrontEndPreview {type=NiagaraBool}
  User.bIsHandHidden {type=NiagaraBool}
  User.bIsHeadHidden {type=NiagaraBool}
 ]
}
