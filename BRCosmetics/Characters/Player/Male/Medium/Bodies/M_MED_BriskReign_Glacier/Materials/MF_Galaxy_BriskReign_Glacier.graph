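# Material function for the BriskReign Glacier player body that outputs a galaxy emissive effect behind the UseGalaxy static switch, with an optional channel mask.
# Two scrolling nebula layers and two star layers use triplanar virtual-box projection for depth.
# Nebulae are hue-shifted, remapped to HDR, screen-blended and time-of-day tinted.
# Stars gain random twinkle and optional recoloring, and a desaturated, contrast-adjusted nebula mask modulates them before they are added to the nebulae.
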MaterialFunction /BRCosmetics/Characters/Player/Male/Medium/Bodies/M_MED_BriskReign_Glacier/Materials/MF_Galaxy_BriskReign_Glacier
graph {nodes [
  n1 = NamedRerouteUsage()
  n2 = NamedRerouteUsage()
  FunctionOutput(None=StaticSwitchParameter(True=Multiply(A=@n1, B=@n2), False=Constant()) {group=Galaxy - General Parameters, parameterName=UseGalaxy, sortPriority=-150}) {outputName="Emissive (V3)"}
  n3 = ScalarParameter() {defaultValue=0.01, group=Galaxy - Stars, parameterName=Galaxy.Stars_TwinkleFrequency_Layer1, sortPriority=15}
  n4 = NamedRerouteUsage()
  n5 = Multiply(
   A=MaterialFunctionCall!(
    Mask=NamedRerouteUsage()
    UVs=NamedRerouteUsage()
    RandomOffset=NamedRerouteUsage()
    BlinkFrequency=Multiply(A=@n3, B=@n4)
    BlinkSpeed=ScalarParameter() {defaultValue=0.7, group=Galaxy - Stars, parameterName=Galaxy.Stars_TwinkleSpeed_Layer1, sortPriority=16}
   ) {materialFunction=/Game/Global/Materials/Character_Parent_2023/RandomBlink_v1/MF_RandomBlink}.Result
   B=ScalarParameter() {defaultValue=5, group=Galaxy - Stars, parameterName=Galaxy.Stars_TwinkleBrightness_Layer1, sortPriority=17}
  )
  Max(A=@n5, B=NamedRerouteUsage())
  n6 = ScalarParameter() {defaultValue=0.02, group=Galaxy - Stars, parameterName=Galaxy.Stars_TwinkleFrequency_Layer2, sortPriority=25}
  n7 = NamedRerouteUsage()
  n8 = Multiply(
   A=MaterialFunctionCall(
    Mask=NamedRerouteUsage()
    UVs=NamedRerouteUsage()
    RandomOffset=NamedRerouteUsage()
    BlinkFrequency=Multiply(A=@n6, B=@n7)
    BlinkSpeed=ScalarParameter() {defaultValue=0.5, group=Galaxy - Stars, parameterName=Galaxy.Stars_TwinkleSpeed_Layer2, sortPriority=26}
   ) {materialFunction=&MF_RandomBlink}.Result
   B=ScalarParameter() {defaultValue=5, group=Galaxy - Stars, parameterName=Galaxy.Stars_TwinkleBrightness_Layer2, sortPriority=27}
  )
  Max(A=@n8, B=NamedRerouteUsage())
  n9 = TextureObjectParameter() {
   group=Galaxy - Stars
   parameterName=Galaxy.StarsTexture
   samplerType=SAMPLERTYPE_LinearColor
   sortPriority=0
   texture=/Game/Global/Materials/Character_Parent_2023/Galaxy_Parent_2025/Textures/T_GalaxyStars
  }
  NamedRerouteDeclaration(None=@n9) {name=StarTex, nodeColor=<rgba 1 0.1641 0 1>}
  n10 = VectorParameter() {group=Galaxy - Stars, parameterName=Glaxy.Stars_Speed_Layer2, sortPriority=23, defaultValue=<rgba -0.001 -0.01 0 1>}
  MaterialExpressionMaterialFunctionCall_4 = MaterialFunctionCall!(
   Texture=NamedRerouteUsage()
   Depth=ScalarParameter() {defaultValue=150, group=Galaxy - Stars, parameterName=Galaxy.Stars_Depth_Layer2, sortPriority=20}
   Tile=ScalarParameter() {defaultValue=3, group=Galaxy - Stars, parameterName=Galaxy.Stars_Tile_Layer2, sortPriority=21}
   BlendContrast=ScalarParameter() {defaultValue=1, group=Galaxy - Stars, parameterName=Galaxy.Stars_BlendContrast_Layer2, sortPriority=22}
   SpeedXY=Multiply(A=ComponentMask(None=@n10.RGB) {g=true, r=true}, B=NamedRerouteUsage())
   ActorOffset=NamedRerouteUsage()
   bUseLocalCameraVector
   bUseDitherSingleSample=StaticBool() {value=true}
  ) {materialFunction=/Game/Global/MaterialFunctions/MF_TriplanarVirtualBox}
  NamedRerouteDeclaration(None=@MaterialExpressionMaterialFunctionCall_4.Result) {name=FarStarsRGB, nodeColor=<rgba 0 0.8125 1 1>}
  n11 = NamedRerouteUsage()
  n12 = ScalarParameter() {defaultValue=1, group=Galaxy - Stars, parameterName=Galaxy.Stars_Brightness_Layer1, sortPriority=14}
  n13 = NamedRerouteUsage()
  n14 = ScalarParameter() {defaultValue=1, group=Galaxy - Stars, parameterName=Galaxy.Stars_Brightness_Layer2, sortPriority=24}
  NamedRerouteDeclaration(None=Max(A=Multiply(A=@n11, B=@n12), B=Multiply(A=@n13, B=@n14))) {name=Stars, nodeColor=<rgba 0 1 0.5078 1>}
  n15 = TextureObjectParameter() {
   group=Galaxy - Nebula
   parameterName=Galaxy.NebulaTexture_Layer1
   sortPriority=10
   texture=/Game/Global/Materials/Character_Parent_2023/Galaxy_Parent_2025/Textures/T_FN_Nebula_style_blue
  }
  n16 = ScalarParameter() {defaultValue=100, group=Galaxy - Nebula, parameterName=Galaxy.Nebula_Depth_Layer1, sortPriority=11}
  n17 = ScalarParameter() {defaultValue=0.4, group=Galaxy - Nebula, parameterName=Galaxy.Nebula_Tile_Layer1, sortPriority=12}
  n18 = ScalarParameter() {defaultValue=0.4, group=Galaxy - Nebula, parameterName=Galaxy.Nebula_BlendContrast_Layer1, sortPriority=13}
  n19 = VectorParameter() {group=Galaxy - Nebula, parameterName=Galaxy.Nebula_Speed_Layer1, sortPriority=14, defaultValue=<rgba 0.002 -0.002 0 1>}
  n20 = Multiply(A=ComponentMask(None=@n19.RGB) {g=true, r=true}, B=NamedRerouteUsage())
  n21 = VectorParameter() {group=Galaxy - General Parameters, parameterName=Galaxy.OriginOffset, sortPriority=10, defaultValue=<rgba 0 0 0 1>}
  n22 = NamedRerouteDeclaration(None=@n21.RGB) {name=ActorOffset, nodeColor=<rgba 0.9688 1 0 1>}
  n23 = Power(
   Base=MaterialFunctionCall(
    Hue Shift Percentage=ScalarParameter() {group=Galaxy - Nebula, parameterName=Galaxy.Nebula_HueShift_Layer1, sortPriority=17}
    Texture=MaterialFunctionCall(Texture=@n15, Depth=@n16, Tile=@n17, BlendContrast=@n18, SpeedXY=@n20, ActorOffset=@n22) {materialFunction=&MF_TriplanarVirtualBox}.Result
   ) {materialFunction=&HueShift}.Result
   Exp=ScalarParameter() {defaultValue=1.5, group=Galaxy - Nebula, parameterName=Galaxy.Nebula_Power_Layer1, sortPriority=16}
  )
  n24 = ScalarParameter() {defaultValue=0.35, group=Galaxy - Nebula, parameterName=Galaxy.Nebula_Desaturation_Layer1, sortPriority=18}
  NamedRerouteDeclaration(None=MaterialFunctionCall(
    LDR Input=Desaturation(None=@n23, Fraction=@n24)
    HDR Intensity Multiplier=ScalarParameter() {defaultValue=0.6, group=Galaxy - Nebula, parameterName=Galaxy.Nebula_Brightness_Layer1, sortPriority=15}
   ) {materialFunction=&DeriveHDRfromLDR}.RemappedRange) {name=Galaxy_01, nodeColor=<rgba 1 0.2578 0 1>}
  n25 = VectorParameter() {group=Galaxy - Stars, parameterName=Glaxy.Stars_Speed_Layer1, sortPriority=13, defaultValue=<rgba -0.001 -0.01 0 1>}
  MaterialExpressionMaterialFunctionCall_11 = MaterialFunctionCall(
   Texture=NamedRerouteUsage()
   Depth=ScalarParameter() {defaultValue=75, group=Galaxy - Stars, parameterName=Galaxy.Stars_Depth_Layer1, sortPriority=10}
   Tile=ScalarParameter() {defaultValue=1.5, group=Galaxy - Stars, parameterName=Galaxy.Stars_Tile_Layer1, sortPriority=11}
   BlendContrast=ScalarParameter() {defaultValue=1, group=Galaxy - Stars, parameterName=Galaxy.Stars_BlendContrast_Layer1, sortPriority=12}
   SpeedXY=Multiply(A=ComponentMask(None=@n25.RGB) {g=true, r=true}, B=NamedRerouteUsage())
   ActorOffset=NamedRerouteUsage()
   bUseDitherSingleSample=StaticBool() {value=true}
  ) {materialFunction=&MF_TriplanarVirtualBox}
  NamedRerouteDeclaration(None=@MaterialExpressionMaterialFunctionCall_11.DitherUVs) {name=CloseStarsUVW, nodeColor=<rgba 1 0 0.7969 1>}
  n26 = Convert(None=@MaterialExpressionMaterialFunctionCall_11.Result) {
   convertInputs [{
     type=Vector3
     expressionInput {
      expression=&self:MaterialExpressionMaterialFunctionCall_11
      inputName=None
      mask=0
      maskA=0
      maskB=0
      maskG=0
      maskR=0
      outputIndex=0
     }
    }]
   convertOutputs [{type=Scalar}, {type=Scalar}, {type=Scalar}]
  }
  NamedRerouteDeclaration(None=NamedRerouteDeclaration(None=@n26) {name=CloseStarsB, nodeColor=<rgba 0.4766 1 0 1>}) {name=CloseStarsRandom, nodeColor=<rgba 1 0.3984 0 1>}
  n27 = Convert(None=@MaterialExpressionMaterialFunctionCall_4.Result) {
   convertInputs [{
     type=Vector3
     expressionInput {
      expression=&self:MaterialExpressionMaterialFunctionCall_4
      inputName=None
      mask=0
      maskA=0
      maskB=0
      maskG=0
      maskR=0
      outputIndex=0
     }
    }]
   convertOutputs [{type=Scalar}, {type=Scalar}, {type=Scalar}]
  }
  NamedRerouteDeclaration(None=NamedRerouteDeclaration(None=@n27) {name=FarStarsB, nodeColor=<rgba 0 0.4375 1 1>}) {name=FarStarsRandom, nodeColor=<rgba 1 0.3984 0 1>}
  n28 = FunctionInput() {bUsePreviewValueAsDefault=true, inputName=TechArtMask, sortPriority=1, previewValue=<1 1 1 1>}
  n29 = ScalarParameter() {defaultValue=1, group=Galaxy - General Parameters, parameterName=Galaxy.Use MaskChannel, sliderMax=1, sortPriority=-100}
  NamedRerouteDeclaration(None=LinearInterpolate(B=ChannelMaskParameter(None=@n28) {group=Galaxy - General Parameters, parameterName=Galaxy.MaskChannel, sortPriority=-50}.R, Alpha=@n29) {constA=1}) {name=GalaxyMask, nodeColor=<rgba 0 1 0.6953 1>}
  NamedRerouteDeclaration(None=@n27) {name=FarStarsR, nodeColor=<rgba 0 1 0.2031 1>}
  NamedRerouteDeclaration(None=@n27) {name=FarStarsG, nodeColor=<rgba 1 0 0.1172 1>}
  NamedRerouteDeclaration(None=@n8) {name=FarStar Twinkle, nodeColor=<rgba 0.1016 1 0 1>}
  n30 = NamedRerouteUsage()
  n31 = VectorParameter() {
   group=Galaxy - General Parameters
   parameterName=Color FarStartR Layer2
   sortPriority=1300
   channelNames {a=Exp, b="", g="", r=""}
   defaultValue=<rgba 1 1 1 1>
  }
  n32 = Power(Base=NamedRerouteUsage(), Exp=@n31.Exp)
  n33 = VectorParameter() {
   group=Galaxy - General Parameters
   parameterName=Color FarStartG Layer2
   sortPriority=1310
   channelNames {a=Exp, b="", g="", r=""}
   defaultValue=<rgba 1 1 1 1>
  }
  n34 = Power(Base=NamedRerouteUsage(), Exp=@n33.Exp)
  n35 = NamedRerouteUsage()
  n36 = VectorParameter() {
   group=Galaxy - General Parameters
   parameterName=Color FarStart Twinkle Layer2
   sortPriority=1330
   channelNames {a=Exp, b="", g="", r=""}
   defaultValue=<rgba 1 1 1 1>
  }
  n37 = Max(A=Max(A=Multiply(A=@n32, B=@n31.RGB), B=Multiply(A=@n34, B=@n33.RGB)), B=Multiply(A=Power(Base=@n35, Exp=@n36.Exp), B=@n36.RGB))
  n38 = ScalarParameter() {group=Galaxy - General Parameters, parameterName=Colorize CloseStars, sliderMax=1, sortPriority=1450}
  n39 = ScalarParameter() {defaultValue=1, group=Galaxy - General Parameters, parameterName=Post Exp CloseStars, sliderMax=10, sliderMin=0.01, sortPriority=1460}
  n40 = Power(
   Base=Multiply(
    A=LinearInterpolate(A=@n30, B=@n37, Alpha=@n38)
    B=ScalarParameter() {defaultValue=1, group=Galaxy - General Parameters, parameterName=Pre Multiply CloseStars, sortPriority=1455}
   )
   Exp=@n39
  )
  n41 = ScalarParameter() {defaultValue=100, group=Galaxy - General Parameters, parameterName=Clamp CloseStars, sortPriority=1480}
  NamedRerouteDeclaration(None=Multiply(
    A=Min(A=@n40, B=@n41)
    B=ScalarParameter() {defaultValue=1, group=Galaxy - General Parameters, parameterName=Post Multiply CloseStars, sortPriority=1490}
   )) {name=Close Stars DEV, nodeColor=<rgba 0.7344 1 0 1>}
  NamedRerouteDeclaration(None=@MaterialExpressionMaterialFunctionCall_4.DitherUVs) {name=FarStarsUVW, nodeColor=<rgba 1 0 0.8906 1>}
  NamedRerouteDeclaration(None=@n26) {name=CloseStarsR, nodeColor=<rgba 0 0.3437 1 1>}
  n42 = NamedRerouteUsage()
  n43 = VectorParameter() {
   group=Galaxy - General Parameters
   parameterName=Color FarStartR Layer2
   sortPriority=1300
   channelNames {a=Exp, b="", g="", r=""}
   defaultValue=<rgba 1 1 1 1>
  }
  n44 = Power(Base=NamedRerouteUsage(), Exp=@n43.Exp)
  n45 = VectorParameter() {
   group=Galaxy - General Parameters
   parameterName=Color FarStartG Layer2
   sortPriority=1310
   channelNames {a=Exp, b="", g="", r=""}
   defaultValue=<rgba 1 1 1 1>
  }
  n46 = Power(Base=NamedRerouteUsage(), Exp=@n45.Exp)
  n47 = NamedRerouteUsage()
  n48 = VectorParameter() {
   group=Galaxy - General Parameters
   parameterName=Color FarStart Twinkle Layer2
   sortPriority=1330
   channelNames {a=Exp, b="", g="", r=""}
   defaultValue=<rgba 1 1 1 1>
  }
  n49 = Max(A=Max(A=Multiply(A=@n44, B=@n43.RGB), B=Multiply(A=@n46, B=@n45.RGB)), B=Multiply(A=Power(Base=@n47, Exp=@n48.Exp), B=@n48.RGB))
  n50 = ScalarParameter() {group=Galaxy - General Parameters, parameterName=Colorize FarStars, sliderMax=1, sortPriority=1350}
  n51 = ScalarParameter() {defaultValue=1, group=Galaxy - General Parameters, parameterName=Post Exp FarStars, sliderMax=10, sliderMin=0.01, sortPriority=1360}
  n52 = Power(
   Base=Multiply(
    A=LinearInterpolate(A=@n42, B=@n49, Alpha=@n50)
    B=ScalarParameter() {defaultValue=1, group=Galaxy - General Parameters, parameterName=Pre Multiply FarStars, sortPriority=1355}
   )
   Exp=@n51
  )
  n53 = ScalarParameter() {defaultValue=100, group=Galaxy - General Parameters, parameterName=Clamp FarStars, sortPriority=1380}
  NamedRerouteDeclaration(None=Multiply(
    A=Min(A=@n52, B=@n53)
    B=ScalarParameter() {defaultValue=1, group=Galaxy - General Parameters, parameterName=Post Multiply FarStars, sortPriority=1390}
   )) {name=Far Stars DEV, nodeColor=<rgba 0.7344 1 0 1>}
  NamedRerouteDeclaration(None=@n5) {name=CloseStar Twinkle, nodeColor=<rgba 0.1016 1 0 1>}
  NamedRerouteDeclaration(None=@MaterialExpressionMaterialFunctionCall_11.Result) {name=CloseStarsRGB, nodeColor=<rgba 0 0.8125 1 1>}
  NamedRerouteDeclaration(None=ScalarParameter() {defaultValue=1, group=Galaxy - General Parameters, parameterName=Global Time Mult, sortPriority=2000}) {name="Global Time Mult (S)", nodeColor=<rgba 0.0078 0 1 1>}
  NamedRerouteDeclaration(None=@n26) {name=CloseStarsG, nodeColor=<rgba 0 0.1562 1 1>}
  n54 = TextureObjectParameter() {
   group=Galaxy - Nebula
   parameterName=Galaxy.NebulaTexture_Layer2
   sortPriority=20
   texture=/Game/Global/Materials/Character_Parent_2023/Galaxy_Parent_2025/Textures/T_FN_Nebula_style2
  }
  n55 = ScalarParameter() {defaultValue=50, group=Galaxy - Nebula, parameterName=Galaxy.Nebula_Depth_Layer2, sortPriority=21}
  n56 = ScalarParameter() {defaultValue=0.4, group=Galaxy - Nebula, parameterName=Galaxy.Nebula_Tile_Layer2, sortPriority=22}
  n57 = ScalarParameter() {defaultValue=1, group=Galaxy - Nebula, parameterName=Galaxy.Nebula_BlendContrast_Layer2, sortPriority=23}
  n58 = VectorParameter() {group=Galaxy - Nebula, parameterName=Galaxy.Nebula_Speed_Layer2, sortPriority=24, defaultValue=<rgba -0.001 0.0005 0 1>}
  n59 = Multiply(A=ComponentMask(None=@n58.RGB) {g=true, r=true}, B=NamedRerouteUsage())
  n60 = NamedRerouteUsage()
  n61 = Power(
   Base=MaterialFunctionCall(
    Hue Shift Percentage=ScalarParameter() {defaultValue=0.05, group=Galaxy - Nebula, parameterName=Galaxy.Nebula_HueShift_Layer2, sortPriority=27}
    Texture=MaterialFunctionCall(Texture=@n54, Depth=@n55, Tile=@n56, BlendContrast=@n57, SpeedXY=@n59, ActorOffset=@n60) {materialFunction=&MF_TriplanarVirtualBox}.Result
   ) {materialFunction=&HueShift}.Result
   Exp=ScalarParameter() {defaultValue=0.85, group=Galaxy - Nebula, parameterName=Galaxy.Nebula_Power_Layer2, sortPriority=26}
  )
  n62 = ScalarParameter() {group=Galaxy - Nebula, parameterName=Galaxy.Nebula_Desaturation_Layer2, sortPriority=28}
  NamedRerouteDeclaration(None=MaterialFunctionCall(
    LDR Input=Desaturation(None=@n61, Fraction=@n62)
    HDR Intensity Multiplier=ScalarParameter() {defaultValue=0.5, group=Galaxy - Nebula, parameterName=Galaxy.Nebula_Brightness_Layer2, sortPriority=25}
   ) {materialFunction=&DeriveHDRfromLDR}.RemappedRange) {name=Galaxy_02, nodeColor=<rgba 0.0781 0 1 1>}
  n63 = Constant() {r=1}
  n64 = Multiply(A=Subtract(A=Constant() {r=1}, B=NamedRerouteUsage()), B=Subtract(A=Constant() {r=1}, B=NamedRerouteUsage()))
  n65 = Desaturation(None=NamedRerouteDeclaration(None=Subtract(A=@n63, B=@n64)) {name=Galaxies, nodeColor=<rgba 1 0 0.2578 1>}, Fraction=Constant() {r=1})
  n66 = ScalarParameter() {defaultValue=0.2, group=Galaxy - Stars, parameterName=Galaxy.Stars_NebulaMask_Power, sortPriority=3}
  NamedRerouteDeclaration(None=MaterialFunctionCall(
    In=Power(Base=@n65, Exp=@n66)
    Contrast=ScalarParameter() {defaultValue=1, group=Galaxy - Stars, parameterName=Galaxy.Stars_NebulaMask_Contrast, sortPriority=4}
   ) {materialFunction=&CheapContrast}.Result) {name=StarMask, nodeColor=<rgba 0.9922 1 0 1>}
  n67 = NamedRerouteUsage()
  n68 = Multiply(
   A=OneMinus(None=ScalarParameter() {group=Galaxy - General Parameters, parameterName=Disable Star Mask, sliderMax=1, sortPriority=1050})
   B=NamedRerouteUsage()
  )
  n69 = VectorParameter() {group=Galaxy - General Parameters, parameterName=Galaxy.TOD_Morning, sortPriority=50, defaultValue=<rgba 1 1 1 1>}
  n70 = VectorParameter() {group=Galaxy - General Parameters, parameterName=Galaxy.TOD_Day, sortPriority=51, defaultValue=<rgba 1 1 1 1>}
  n71 = VectorParameter() {group=Galaxy - General Parameters, parameterName=Galaxy.TOD_Evening, sortPriority=52, defaultValue=<rgba 1 1 1 1>}
  n72 = VectorParameter() {group=Galaxy - General Parameters, parameterName=Galaxy.TOD_Night, sortPriority=53, defaultValue=<rgba 1 1 1 1>}
  n73 = Multiply(
   A=NamedRerouteUsage()
   B=MaterialFunctionCall(Color_Morning=@n69.RGB, Color_Day=@n70.RGB, Color_Evening=@n71.RGB, Color_Night=@n72.RGB) {materialFunction=&Color_Time_of_Day}.Result
  ) {constA=1}
  n74 = Multiply(
   A=Saturate(None=LinearInterpolate(A=@n67, Alpha=@n68) {constB=0})
   B=Multiply(A=@n73, B=ScalarParameter() {defaultValue=25, group=Galaxy - Stars, parameterName=Galaxy.Stars_GlobalBrightness, sortPriority=1}) {constB=20}
  )
  n75 = ScalarParameter() {defaultValue=0.5, group=Galaxy - Stars, parameterName=Galaxy.Stars_Desaturation, sortPriority=2}
  n76 = ScalarParameter() {defaultValue=1, group=Galaxy - General Parameters, parameterName=Nebula Mix, sliderMax=1, sortPriority=1000}
  NamedRerouteDeclaration(None=Add(A=Desaturation(None=@n74, Fraction=@n75), B=Multiply(A=@n76, B=@n73)) {constB=0}) {name=GalaxiesAndStars, nodeColor=<rgba 0 1 0.9063 1>}
  Reroute(None=@n73)
 ]}
