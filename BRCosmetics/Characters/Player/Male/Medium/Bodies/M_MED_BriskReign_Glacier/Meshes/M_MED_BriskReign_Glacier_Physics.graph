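# Physics asset for the Glacier variant of the male medium BriskReign body, setting up its coat bones for dynamic secondary motion.
# Back and side chains hang from the pelvis and front chains from spine_01, with swing and twist limits below free-moving base joints.
# Free constraints cross-link left and right back and front chains in both directions.
# Leg, foot and torso capsules provide body shapes, and a huge offset root sphere appears to approximate the ground.
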PhysicsAsset /BRCosmetics/Characters/Player/Male/Medium/Bodies/M_MED_BriskReign_Glacier/Meshes/M_MED_BriskReign_Glacier_Physics
bodies {
 base_coat_bk_l [{shapeType=Capsule, length=1, radius=1}]
 base_coat_bk_r [{shapeType=Capsule, length=1, radius=1}]
 base_coat_fr_l [{shapeType=Capsule, length=0.505, radius=0.505, shapeName=base_coat_fr_l_capsule, rotation=<rot 0 -90 0>}]
 base_coat_fr_r [{shapeType=Capsule, length=0.505, radius=0.505, shapeName=base_coat_fr_r_capsule, rotation=<rot 0 -90 0>}]
 base_coat_pad_main_bk [{shapeType=Capsule, length=1, radius=1}]
 base_coat_pad_main_fr [{shapeType=Capsule, length=1, radius=1}]
 base_coat_side_l [{shapeType=Capsule, length=0.505, radius=0.505, shapeName=base_coat_side_l_capsule, rotation=<rot 0 -90 0>}]
 base_coat_side_r [{shapeType=Capsule, length=0.505, radius=0.505, shapeName=base_coat_side_r_capsule, rotation=<rot 0 -90 0>}]
 calf_l [{shapeType=Capsule, length=22, radius=12, shapeName=calf_l_capsule, center=<-10.8116 0 0>, rotation=<rot -90 0 0>}]
 calf_r [{shapeType=Capsule, length=22, radius=12, shapeName=calf_r_capsule, center=<10.8116 0 0>, rotation=<rot -90 0 0>}]
 dyn_coat_bk_01_l [{shapeType=Capsule, length=12, radius=5, center=<5 0 6>, rotation=<rot -90 0 0>}]
 dyn_coat_bk_01_r [{shapeType=Capsule, length=12, radius=5, center=<5 0 6>, rotation=<rot -90 0 0>}]
 dyn_coat_bk_02_l [{shapeType=Capsule, length=12, radius=6, center=<5 0 6>, rotation=<rot -90 0 0>}]
 dyn_coat_bk_02_r [{shapeType=Capsule, length=12, radius=6, center=<5 0 6>, rotation=<rot -90 0 0>}]
 dyn_coat_bk_03_l [{shapeType=Capsule, length=12, radius=6, center=<5 0 4>, rotation=<rot -90 0 0>}]
 dyn_coat_bk_03_r [{shapeType=Capsule, length=12, radius=6, center=<5 0 4.9195>, rotation=<rot -90 0 0>}]
 dyn_coat_bk_04_l [{shapeType=Capsule, length=12, radius=6, center=<5 0 4>, rotation=<rot -90 0 0>}]
 dyn_coat_bk_04_r [{shapeType=Capsule, length=12, radius=6, center=<5 0 4>, rotation=<rot -90 0 0>}]
 dyn_coat_fr_l_1 [{shapeType=Capsule, length=12, radius=6, center=<15 0 6>, rotation=<rot -90 0 0>}]
 dyn_coat_fr_l_2 [{shapeType=Capsule, length=12, radius=6, center=<15 0 6>, rotation=<rot -90 0 0>}]
 dyn_coat_fr_r_1 [{shapeType=Capsule, length=12, radius=6, center=<15 0 6>, rotation=<rot -90 0 0>}]
 dyn_coat_fr_r_2 [{shapeType=Capsule, length=12, radius=6, center=<15 0 6>, rotation=<rot -90 0 0>}]
 dyn_coat_pad_main_bk [{shapeType=Capsule, length=1, radius=1}]
 dyn_coat_pad_main_fr [{shapeType=Capsule, length=1, radius=1}]
 dyn_coat_side_01_l [{shapeType=Capsule, length=12, radius=10, center=<6 0 10.5>, rotation=<rot -90 0 90>}]
 dyn_coat_side_01_r [{shapeType=Capsule, length=12, radius=10, center=<6 0 10.5>, rotation=<rot -90 0 90>}]
 dyn_coat_side_02_l [{shapeType=Capsule, length=12, radius=10, center=<6 0 10.5>, rotation=<rot -90 0 90>}]
 dyn_coat_side_02_r [{shapeType=Capsule, length=12, radius=10, center=<6 0 10.5>, rotation=<rot -90 0 90>}]
 dyn_coat_side_03_l [{shapeType=Capsule, length=12, radius=10, center=<6 0 10>, rotation=<rot -90 0 90>}]
 dyn_coat_side_03_r [{shapeType=Capsule, length=12, radius=10, center=<6 0 10>, rotation=<rot -90 0 90>}]
 dyn_coat_side_04_l [{shapeType=Capsule, length=12, radius=10, center=<6 0 10>, rotation=<rot -90 0 90>}]
 dyn_coat_side_04_r [{shapeType=Capsule, length=12, radius=10, center=<6 0 10>, rotation=<rot -90 0 90>}]
 dyn_coat_side_05_l [{shapeType=Capsule, length=12, radius=10, center=<6 0 10>, rotation=<rot -90 0 90>}]
 dyn_coat_side_05_r [{shapeType=Capsule, length=12, radius=10, center=<6 0 9.8166>, rotation=<rot -90 90 0>}]
 foot_l [{shapeType=Capsule, length=11.0044, radius=11.0044, shapeName=foot_l_capsule, center=<-5 8 0>, rotation=<rot 0 0 -90>}]
 foot_r [{shapeType=Capsule, length=11.0044, radius=11.0044, shapeName=foot_r_capsule, center=<5 -8 0>, rotation=<rot 0 0 -90>}]
 pelvis [{shapeType=Capsule, length=9.6126, radius=10.6126, shapeName=pelvis_capsule, center=<0 -1.3184 0>}]
 root [{shapeType=Sphere, radius=500, shapeName=root_sphere, center=<0 17.4691 -491.314>}]
 spine_01 [{shapeType=Capsule, length=8.6209, radius=10.6209, shapeName=spine_01_capsule}]
 thigh_l [{shapeType=Capsule, length=22, radius=12, shapeName=thigh_l_capsule, center=<-14.5176 -0.3653 0>, rotation=<rot -88.3183 0 0>}]
 thigh_r [{shapeType=Capsule, length=22, radius=12, shapeName=thigh_r_capsule, center=<14.5176 0.3653 0>, rotation=<rot -88.3183 0 0>}]
}
constraints [
 {
  bone1Name=base_coat_pad_main_bk
  bone2Name=pelvis
  swing1LimitDegrees=45
  swing1Motion=Free
  swing2LimitDegrees=45
  swing2Motion=Free
  twistLimitDegrees=45
  twistMotion=Free
 }
 {
  bone1Name=dyn_coat_pad_main_bk
  bone2Name=base_coat_pad_main_bk
  swing1LimitDegrees=45
  swing1Motion=Free
  swing2LimitDegrees=45
  swing2Motion=Free
  twistLimitDegrees=45
  twistMotion=Free
 }
 {
  bone1Name=base_coat_bk_l
  bone2Name=dyn_coat_pad_main_bk
  swing1LimitDegrees=45
  swing1Motion=Free
  swing2LimitDegrees=45
  swing2Motion=Free
  twistLimitDegrees=45
  twistMotion=Free
 }
 {bone1Name=dyn_coat_bk_01_l, bone2Name=base_coat_bk_l, swing1LimitDegrees=35, swing2LimitDegrees=35, twistLimitDegrees=1}
 {bone1Name=dyn_coat_bk_02_l, bone2Name=dyn_coat_bk_01_l, swing1LimitDegrees=35, swing2LimitDegrees=35, twistLimitDegrees=1}
 {bone1Name=dyn_coat_bk_03_l, bone2Name=dyn_coat_bk_02_l, swing1LimitDegrees=35, swing2LimitDegrees=35, twistLimitDegrees=1}
 {bone1Name=dyn_coat_bk_04_l, bone2Name=dyn_coat_bk_03_l, swing1LimitDegrees=35, swing2LimitDegrees=35, twistLimitDegrees=1}
 {
  bone1Name=base_coat_bk_r
  bone2Name=dyn_coat_pad_main_bk
  swing1LimitDegrees=45
  swing1Motion=Free
  swing2LimitDegrees=45
  swing2Motion=Free
  twistLimitDegrees=45
  twistMotion=Free
 }
 {bone1Name=dyn_coat_bk_01_r, bone2Name=base_coat_bk_r, swing1LimitDegrees=35, swing2LimitDegrees=35, twistLimitDegrees=1}
 {bone1Name=dyn_coat_bk_02_r, bone2Name=dyn_coat_bk_01_r, swing1LimitDegrees=35, swing2LimitDegrees=35, twistLimitDegrees=1}
 {bone1Name=dyn_coat_bk_03_r, bone2Name=dyn_coat_bk_02_r, swing1LimitDegrees=35, swing2LimitDegrees=35, twistLimitDegrees=1}
 {bone1Name=dyn_coat_bk_04_r, bone2Name=dyn_coat_bk_03_r, swing1LimitDegrees=35, swing2LimitDegrees=35, twistLimitDegrees=1}
 {
  bone1Name=base_coat_pad_main_fr
  bone2Name=spine_01
  swing1LimitDegrees=45
  swing1Motion=Free
  swing2LimitDegrees=45
  swing2Motion=Free
  twistLimitDegrees=45
  twistMotion=Free
 }
 {
  bone1Name=dyn_coat_pad_main_fr
  bone2Name=base_coat_pad_main_fr
  swing1LimitDegrees=45
  swing1Motion=Free
  swing2LimitDegrees=45
  swing2Motion=Free
  twistLimitDegrees=45
  twistMotion=Free
 }
 {
  bone1Name=dyn_coat_bk_02_r
  bone2Name=dyn_coat_bk_02_l
  swing1LimitDegrees=45
  swing1Motion=Free
  swing2LimitDegrees=45
  swing2Motion=Free
  twistLimitDegrees=45
  twistMotion=Free
 }
 {
  bone1Name=dyn_coat_bk_02_l
  bone2Name=dyn_coat_bk_02_r
  swing1LimitDegrees=45
  swing1Motion=Free
  swing2LimitDegrees=45
  swing2Motion=Free
  twistLimitDegrees=45
  twistMotion=Free
 }
 {
  bone1Name=base_coat_side_r
  bone2Name=pelvis
  swing1LimitDegrees=45
  swing1Motion=Free
  swing2LimitDegrees=45
  swing2Motion=Free
  twistLimitDegrees=45
  twistMotion=Free
 }
 {bone1Name=dyn_coat_side_01_r, bone2Name=base_coat_side_r, swing1LimitDegrees=15, swing2LimitDegrees=45, twistLimitDegrees=25}
 {bone1Name=dyn_coat_side_03_r, bone2Name=dyn_coat_side_02_r, swing1LimitDegrees=45, swing2LimitDegrees=45, twistLimitDegrees=25}
 {bone1Name=dyn_coat_side_04_r, bone2Name=dyn_coat_side_03_r, swing1LimitDegrees=45, swing2LimitDegrees=45, twistLimitDegrees=25}
 {bone1Name=dyn_coat_side_05_r, bone2Name=dyn_coat_side_04_r, swing1LimitDegrees=45, swing2LimitDegrees=45, twistLimitDegrees=25}
 {
  bone1Name=base_coat_side_l
  bone2Name=pelvis
  swing1LimitDegrees=45
  swing1Motion=Free
  swing2LimitDegrees=45
  swing2Motion=Free
  twistLimitDegrees=45
  twistMotion=Free
 }
 {bone1Name=dyn_coat_side_01_l, bone2Name=base_coat_side_l, swing1LimitDegrees=15, swing2LimitDegrees=45, twistLimitDegrees=25}
 {bone1Name=dyn_coat_side_03_l, bone2Name=dyn_coat_side_02_l, swing1LimitDegrees=45, swing2LimitDegrees=45, twistLimitDegrees=25}
 {bone1Name=dyn_coat_side_04_l, bone2Name=dyn_coat_side_03_l, swing1LimitDegrees=45, swing2LimitDegrees=45, twistLimitDegrees=25}
 {bone1Name=dyn_coat_side_05_l, bone2Name=dyn_coat_side_04_l, swing1LimitDegrees=45, swing2LimitDegrees=45, twistLimitDegrees=25}
 {bone1Name=dyn_coat_side_02_r, bone2Name=dyn_coat_side_01_r, swing1LimitDegrees=45, swing2LimitDegrees=45, twistLimitDegrees=25}
 {bone1Name=dyn_coat_side_02_l, bone2Name=dyn_coat_side_01_l, swing1LimitDegrees=45, swing2LimitDegrees=45, twistLimitDegrees=25}
 {
  bone1Name=base_coat_fr_l
  bone2Name=dyn_coat_pad_main_fr
  swing1LimitDegrees=45
  swing1Motion=Free
  swing2LimitDegrees=45
  swing2Motion=Free
  twistLimitDegrees=45
  twistMotion=Free
 }
 {bone1Name=dyn_coat_fr_l_1, bone2Name=base_coat_fr_l, swing1LimitDegrees=10, swing2LimitDegrees=60, twistLimitDegrees=25}
 {bone1Name=dyn_coat_fr_l_2, bone2Name=dyn_coat_fr_l_1, swing1LimitDegrees=10, swing2LimitDegrees=60, twistLimitDegrees=25}
 {
  bone1Name=base_coat_fr_r
  bone2Name=dyn_coat_pad_main_fr
  swing1LimitDegrees=45
  swing1Motion=Free
  swing2LimitDegrees=45
  swing2Motion=Free
  twistLimitDegrees=45
  twistMotion=Free
 }
 {bone1Name=dyn_coat_fr_r_1, bone2Name=base_coat_fr_r, swing1LimitDegrees=10, swing2LimitDegrees=60, twistLimitDegrees=25}
 {bone1Name=dyn_coat_fr_r_2, bone2Name=dyn_coat_fr_r_1, swing1LimitDegrees=10, swing2LimitDegrees=60, twistLimitDegrees=25}
 {
  bone1Name=dyn_coat_fr_l_2
  bone2Name=dyn_coat_fr_r_2
  swing1LimitDegrees=45
  swing1Motion=Free
  swing2LimitDegrees=45
  swing2Motion=Free
  twistLimitDegrees=45
  twistMotion=Free
 }
 {
  bone1Name=dyn_coat_fr_r_2
  bone2Name=dyn_coat_fr_l_2
  swing1LimitDegrees=45
  swing1Motion=Free
  swing2LimitDegrees=45
  swing2Motion=Free
  twistLimitDegrees=45
  twistMotion=Free
 }
]
