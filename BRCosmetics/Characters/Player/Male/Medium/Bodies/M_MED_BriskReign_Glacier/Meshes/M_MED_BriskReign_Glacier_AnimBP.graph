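# Post-pose Animation Blueprint for the BriskReign Glacier medium male body, adding coat dynamics and chest-pad correction.
# Pose copied from the attached parent feeds a slide/crouch-aware Control Rig, pelvis-space RigidBody with parameter-driven gravity, then thigh-to-back-coat rotation constraints.
# Capes or long back blings halve physics and partly enable those constraints.
# Chest-pad pitch follows head-down angle, fixed in the front end outside emotes; side-coat bone nodes are unconnected.
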AnimBlueprint /BRCosmetics/Characters/Player/Male/Medium/Bodies/M_MED_BriskReign_Glacier/Meshes/M_MED_BriskReign_Glacier_AnimBP
target_skeleton=&/M_MED_BriskReign_Glacier_Skeleton
graphs [
 AnimGraph {nodes [
   Local To Component()
   n1 = "Transform (Modify) Bone - Bone: base_coat_side_r"!(ComponentPose, bAlphaBoolEnabled:bool=True, Translation="0,0,0", Rotation="5,0,0", Scale="1,1,1") -> (Pose) {
    alphaInputType=Bool
    rotationMode=BMM_Additive
    rotationSpace=BCS_BoneSpace
    alphaBoolBlend {blendInTime=0.2, blendOutTime=0.2}
    boneToModify {boneName=base_coat_side_r}
    rotation {pitch=5}
   }
   "Transform (Modify) Bone - Bone: base_coat_side_l"!(ComponentPose=@n1, bAlphaBoolEnabled:bool=True, Translation="0,0,0", Rotation="5,0,0", Scale="1,1,1") -> (Pose) {
    alphaInputType=Bool
    rotationMode=BMM_Additive
    rotationSpace=BCS_BoneSpace
    alphaBoolBlend {blendInTime=0.2, blendOutTime=0.2}
    boneToModify {boneName=base_coat_side_l}
    rotation {pitch=5}
   }
   n2 = Control Rig!(
    Source=Copy Pose From Mesh() {bUseAttachedParent=true}
    Alpha
    isSliding:bool=Get bIsSliding()
    isCrouching:bool=Get bIsCrouching()
    isSLidingFloat:double=Get isSliding() -> (isSliding:double)
   ) -> (Pose) {controlRigAssetReference {blueprintRigClass=&/M_MED_BriskReign_Glacier_CtrlRig.M_MED_BriskReign_Glacier_CtrlRig_C}}
   n3 = RigidBody!(
    ComponentPose=Local To Component(LocalPose=@n2)
    Alpha=Get RB_Value() -> (RB_Value:double)
    OverrideWorldGravity:Vector=Get GravityOverride()
    ExternalForce
   ) -> (Pose) {
    bClampLinearTranslationLimitToRefPose=true
    bOverrideWorldGravity=true
    overridePhysicsAsset=&/M_MED_BriskReign_Glacier_Physics
    simulationSpace=BaseBoneSpace
    alphaBoolBlend {blendOption=Linear}
    baseBoneRef {boneName=pelvis}
    overrideWorldGravity {z=-980}
    simSpaceSettings {dampingAlpha=0.25, maxAngularAcceleration=8, maxAngularVelocity=2.5, maxLinearAcceleration=800, maxLinearVelocity=300, worldAlpha=1}
   }
   n4 = "Constraint - dyn_coat_bk_01_l (R)"(ComponentPose=@n3, Alpha=Get constraintsAlpha() -> (constraintsAlpha:double), ConstraintWeights_0=1) {boneToModify {boneName=dyn_coat_bk_01_l}, constraintSetup [{transformType=Rotation, targetBone {boneName=thigh_l}}], constraintWeights [1]}
   n5 = "Constraint - dyn_coat_bk_01_r (R)"(ComponentPose=@n4, Alpha=Get constraintsAlpha(), ConstraintWeights_0=1) {boneToModify {boneName=dyn_coat_bk_01_r}, constraintSetup [{transformType=Rotation, targetBone {boneName=thigh_r}}], constraintWeights [1]}
   n6 = "Transform (Modify) Bone - Bone: base_chest_pad"(ComponentPose=@n5, Alpha=Get Head Down() -> (Head Down:double), Translation="0,0,0", Rotation="35,0,0", Scale="1,1,1") {
    rotationMode=BMM_Additive
    rotationSpace=BCS_BoneSpace
    translationMode=BMM_Additive
    translationSpace=BCS_BoneSpace
    boneToModify {boneName=base_chest_pad}
    rotation {pitch=35}
   }
   Output Pose(Result=Component To Local(ComponentPose=@n6)) {layerGroup=None}
  ]}
 EventGraph {nodes [
   n1 = Event BlueprintUpdateAnimation()
   > Set delta(delta:double=@n1.DeltaTimeX) -> (Output_Get:double)
   > n2 = Sequence()
   > n3 = "Gravity Override State Function (Parameter Asset) (Nativized)"(
    CharacterPartAnimInstance=Self-Reference()
    GameTimeInSeconds=GetGameTimeInSeconds()
    Parameters=&/M_MED_BriskReign_Glacier_Parameters
   )
   > Set GravityOverride(GravityOverride=@n3)
   > Set RBAN Settings Out(RBAN Settings Out=@n3.OutRigidBodyAnimNodeParameters)
   n4 = Get bIsSliding()
   > Set isSliding(isSliding:double=FInterpTo(Current=Get isSliding(), Target="To Float (Boolean)"(InBool=@n4), DeltaTime=Get delta() -> (delta:double), InterpSpeed=5)) -> (Output_Get:double)
   > n5 = Set Head Down(Head Down:double=GetAlphaFromReferenceTransformDelta(SkeletalMeshComponent=Get Skel Mesh(), BoneName=spine_05, BaseName=head, InputRange=(X=0,Y=-45), TwistAxis="0,0,1")) -> (Output_Get:double)
   n6 = Get bIsPlayingEmote()
   > n7 = Branch(Condition=AND Boolean(A=Get bIsInFrontEnd(), B=NOT Boolean(A=@n6)))
   > Set Head Down(Head Down=0.7)
   Set Head Down(execute=@n7.else, Head Down=@n5)
   n8 = CheckForBackbling()
   n9 = Set Is Wearing Cape(execute=@n2.then_1, Is Wearing Cape=OR Boolean(A=@n8.bIsBackblingACape, B=@n8.bIsBackblingLong))
   > n10 = Branch(Condition=@n9)
   > Set RB_Value(RB_Value:double=0.5) -> (Output_Get:double)
   > Set constraintsAlpha(constraintsAlpha:double=0.2) -> (Output_Get:double)
   Set RB_Value(execute=@n10.else, RB_Value=1)
   > Set constraintsAlpha(constraintsAlpha=0)
   Event BlueprintInitializeAnimation()
   > Set Skel Mesh(Skel Mesh=GetOwnerSkeletalMeshComponent())
   Set isWearingBackbling(isWearingBackbling=AND Boolean!(A=@n8, B=NOT Boolean(A=@n8.bIsBackblingACape), C:bool=NOT Boolean(A=@n8.bIsBackblingLong)) -> (ReturnValue))
  ]}
]
variables [
 constraintsAlpha {type=number}
 delta {type=number}
 GravityOverride {type=Vector}
 Head Down {type=number}
 Is Wearing Cape {type=boolean}
 isSliding {type=number}
 isWearingBackbling {type=boolean}
 RB_Value {type=number}
 RBAN Settings Out {type=FortRigidBodyAnimNodeParameters}
 Skel Mesh {type=SkeletalMeshComponent}
]
