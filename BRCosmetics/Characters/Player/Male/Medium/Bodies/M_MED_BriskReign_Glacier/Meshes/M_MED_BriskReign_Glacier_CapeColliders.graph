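# Physics asset of cape colliders for the male medium BriskReign Glacier character body.
# Capsules cover the head, clavicles, spine, pelvis, feet and dyn_coat bone chains, while arm, leg and neck bodies hold no shapes.
# A huge downward-offset root capsule, named as a sphere, forms a near-flat surface at the character's base.
# Constraints link only core skeleton bodies with free swing and twist, leaving coat bodies unconstrained.
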PhysicsAsset /BRCosmetics/Characters/Player/Male/Medium/Bodies/M_MED_BriskReign_Glacier/Meshes/M_MED_BriskReign_Glacier_CapeColliders
bodies {
 calf_l []
 calf_r []
 clavicle_l [{shapeType=Capsule, length=7.6964, radius=6.3853, center=<4.0147 -0.8354 -1.7419>, rotation=<rot -77.5566 177.3915 7.7314>}]
 clavicle_r [{shapeType=Capsule, length=7.6964, radius=8.2957, center=<-4.0147 0.8354 1.7419>, rotation=<rot -77.5565 177.3915 -172.2685>}]
 dyn_coat_bk_01 [{shapeType=Capsule, length=4.75, radius=17, shapeName=dyn_coat_bk_l_01_capsule, center=<13.8814 0 -10.8757>, rotation=<rot 90 0 0>}]
 dyn_coat_bk_02 [{shapeType=Capsule, length=5.75, radius=21.25, shapeName=dyn_coat_bk_02_capsule, center=<16.522 0 -15.5817>, rotation=<rot 90 0 0>}]
 dyn_coat_bk_03 [{shapeType=Capsule, length=10, radius=20, shapeName=dyn_coat_bk_03_capsule, center=<16.522 0 -11.3992>, rotation=<rot 90 0 0>}]
 dyn_coat_bk_04 [{shapeType=Capsule, length=11, radius=19, shapeName=dyn_coat_bk_04_capsule, center=<16.522 0.0419 -11.115>, rotation=<rot 90 0 0>}]
 dyn_coat_bk_l_01 [{shapeType=Capsule, length=4.75, radius=17, shapeName=dyn_coat_bk_l_01_capsule, center=<13.8814 0 -10.8757>, rotation=<rot 90 0 0>}]
 dyn_coat_bk_l_02 [{shapeType=Capsule, length=5.75, radius=21.25, shapeName=dyn_coat_bk_l_02_capsule, center=<16.522 0 -15.5817>, rotation=<rot 90 0 0>}]
 dyn_coat_bk_l_03 [{shapeType=Capsule, length=10, radius=20, shapeName=dyn_coat_bk_l_03_capsule, center=<16.522 0 -11.3992>, rotation=<rot 90 0 0>}]
 dyn_coat_bk_l_04 [{shapeType=Capsule, length=6.75, radius=23.25, shapeName=dyn_coat_bk_l_04_capsule, center=<16.522 0.0419 -6.2786>, rotation=<rot 90 0 0>}]
 dyn_coat_bk_r_01 [{shapeType=Capsule, length=4.75, radius=17, shapeName=dyn_coat_bk_l_01_capsule, center=<13.8814 0 -10.8757>, rotation=<rot 90 0 0>}]
 dyn_coat_bk_r_02 [{shapeType=Capsule, length=5.75, radius=21.25, shapeName=dyn_coat_bk_r_02_capsule, center=<16.522 0 -15.5817>, rotation=<rot 90 0 0>}]
 dyn_coat_bk_r_03 [{shapeType=Capsule, length=10, radius=20, shapeName=dyn_coat_bk_r_03_capsule, center=<16.522 0 -11.3992>, rotation=<rot 90 0 0>}]
 dyn_coat_bk_r_04 [{shapeType=Capsule, length=8, radius=22, shapeName=dyn_coat_bk_r_04_capsule, center=<16.522 0.0419 -5.3573>, rotation=<rot 90 0 0>}]
 dyn_coat_l_01 [{shapeType=Capsule, length=19, radius=8, shapeName=dyn_coat_r_01_capsule, center=<8.1765 0 -3.3698>, rotation=<rot 90 0 0>}]
 dyn_coat_l_02 [{shapeType=Capsule, length=5.75, radius=21.25, shapeName=dyn_coat_bk_l_02_capsule, center=<16.522 0 -15.5817>, rotation=<rot 90 0 0>}]
 dyn_coat_l_03 [{shapeType=Capsule, length=10, radius=20, shapeName=dyn_coat_bk_l_03_capsule, center=<16.522 0 -11.3992>, rotation=<rot 90 0 0>}]
 dyn_coat_r_01 [{shapeType=Capsule, length=19, radius=8, shapeName=dyn_coat_r_01_capsule, center=<8.1765 0 -3.3698>, rotation=<rot 90 0 0>}]
 dyn_coat_r_02 [{shapeType=Capsule, length=5.75, radius=21.25, shapeName=dyn_coat_bk_l_02_capsule, center=<16.522 0 -15.5817>, rotation=<rot 90 0 0>}]
 dyn_coat_r_03 [{shapeType=Capsule, length=10, radius=20, shapeName=dyn_coat_bk_l_03_capsule, center=<16.522 0 -11.3992>, rotation=<rot 90 0 0>}]
 foot_l [{shapeType=Capsule, length=12.5982, radius=6.5127, center=<-2.2072 6.0534 -0.3367>, rotation=<rot 0.0388 1.4701 -89.3982>}]
 foot_r [{shapeType=Capsule, length=12.5982, radius=6.5127, center=<2.2072 -6.0534 0.3367>, rotation=<rot 0.0388 1.4701 90.6018>}]
 head [{shapeType=Capsule, length=0, radius=18, center=<9.0697 0.0746 0>, rotation=<rot 0 -90 0>}]
 lowerarm_l []
 lowerarm_r []
 neck_02 []
 pelvis [{shapeType=Capsule, length=9.5, radius=15, shapeName=pelvis_capsule, center=<-1.5557 -7.7308 0>}]
 root [{shapeType=Capsule, length=3, radius=298, shapeName=root_sphere, center=<0 -24.5697 -289.8553>, rotation=<rot 0 -90 0>}]
 spine_01 [
  {shapeType=Capsule, length=15.0484, radius=8.4854, center=<1.179 2.2974 0>}
  {shapeType=Capsule, length=20.6875, radius=9.3125, center=<-1.8317 2.606 0>, rotation=<rot 90 0 0>}
 ]
 spine_02 [
  {shapeType=Capsule, length=7.7984, radius=13.4854, center=<1.179 -4.9617 0>}
  {shapeType=Capsule, length=20.6875, radius=9.3125, center=<-1.8317 2.606 0>, rotation=<rot 90 0 0>}
 ]
 spine_03 [
  {shapeType=Capsule, length=4.2984, radius=13.4854, center=<-0.0747 -0.8086 0>}
  {shapeType=Capsule, length=20.7188, radius=9.2812, center=<-1.2326 2.6059 0>, rotation=<rot 90 0 0>}
 ]
 spine_04 [
  {shapeType=Capsule, length=0.8296, radius=15.3916, center=<5.5182 -4.5368 0>}
  {shapeType=Capsule, length=9.7812, radius=9.5312, center=<-1.2326 1.308 0>, rotation=<rot 90 0 0>}
 ]
 spine_05 [{shapeType=Capsule, length=5.9859, radius=15.0791, center=<5.6377 -2.8584 0>}]
 thigh_l []
 thigh_r []
 upperarm_l []
 upperarm_r []
}
constraints [
 {
  bone1Name=spine_01
  bone2Name=pelvis
  swing1LimitDegrees=45
  swing1Motion=Free
  swing2LimitDegrees=45
  swing2Motion=Free
  twistLimitDegrees=45
  twistMotion=Free
 }
 {
  bone1Name=spine_02
  bone2Name=spine_01
  swing1LimitDegrees=45
  swing1Motion=Free
  swing2LimitDegrees=45
  swing2Motion=Free
  twistLimitDegrees=45
  twistMotion=Free
 }
 {
  bone1Name=spine_03
  bone2Name=spine_02
  swing1LimitDegrees=45
  swing1Motion=Free
  swing2LimitDegrees=45
  swing2Motion=Free
  twistLimitDegrees=45
  twistMotion=Free
 }
 {
  bone1Name=spine_04
  bone2Name=spine_03
  swing1LimitDegrees=45
  swing1Motion=Free
  swing2LimitDegrees=45
  swing2Motion=Free
  twistLimitDegrees=45
  twistMotion=Free
 }
 {
  bone1Name=spine_05
  bone2Name=spine_04
  swing1LimitDegrees=45
  swing1Motion=Free
  swing2LimitDegrees=45
  swing2Motion=Free
  twistLimitDegrees=45
  twistMotion=Free
 }
 {
  bone1Name=clavicle_l
  bone2Name=spine_05
  swing1LimitDegrees=45
  swing1Motion=Free
  swing2LimitDegrees=45
  swing2Motion=Free
  twistLimitDegrees=45
  twistMotion=Free
 }
 {
  bone1Name=upperarm_l
  bone2Name=clavicle_l
  swing1LimitDegrees=45
  swing1Motion=Free
  swing2LimitDegrees=45
  swing2Motion=Free
  twistLimitDegrees=45
  twistMotion=Free
 }
 {
  bone1Name=lowerarm_l
  bone2Name=upperarm_l
  swing1LimitDegrees=45
  swing1Motion=Free
  swing2LimitDegrees=45
  swing2Motion=Free
  twistLimitDegrees=45
  twistMotion=Free
 }
 {
  bone1Name=neck_02
  bone2Name=spine_05
  swing1LimitDegrees=45
  swing1Motion=Free
  swing2LimitDegrees=45
  swing2Motion=Free
  twistLimitDegrees=45
  twistMotion=Free
 }
 {
  bone1Name=head
  bone2Name=neck_02
  swing1LimitDegrees=45
  swing1Motion=Free
  swing2LimitDegrees=45
  swing2Motion=Free
  twistLimitDegrees=45
  twistMotion=Free
 }
 {
  bone1Name=foot_l
  bone2Name=calf_l
  swing1LimitDegrees=45
  swing1Motion=Free
  swing2LimitDegrees=45
  swing2Motion=Free
  twistLimitDegrees=45
  twistMotion=Free
 }
 {
  bone1Name=clavicle_r
  bone2Name=spine_05
  swing1LimitDegrees=45
  swing1Motion=Free
  swing2LimitDegrees=45
  swing2Motion=Free
  twistLimitDegrees=45
  twistMotion=Free
 }
 {
  bone1Name=lowerarm_r
  bone2Name=upperarm_r
  swing1LimitDegrees=45
  swing1Motion=Free
  swing2LimitDegrees=45
  swing2Motion=Free
  twistLimitDegrees=45
  twistMotion=Free
 }
 {
  bone1Name=foot_r
  bone2Name=calf_r
  swing1LimitDegrees=45
  swing1Motion=Free
  swing2LimitDegrees=45
  swing2Motion=Free
  twistLimitDegrees=45
  twistMotion=Free
 }
 {
  bone1Name=upperarm_r
  bone2Name=clavicle_r
  swing1LimitDegrees=45
  swing1Motion=Free
  swing2LimitDegrees=45
  swing2Motion=Free
  twistLimitDegrees=45
  twistMotion=Free
 }
 {
  bone1Name=thigh_l
  bone2Name=pelvis
  swing1LimitDegrees=45
  swing1Motion=Free
  swing2LimitDegrees=45
  swing2Motion=Free
  twistLimitDegrees=45
  twistMotion=Free
 }
 {
  bone1Name=thigh_r
  bone2Name=pelvis
  swing1LimitDegrees=45
  swing1Motion=Free
  swing2LimitDegrees=45
  swing2Motion=Free
  twistLimitDegrees=45
  twistMotion=Free
 }
 {
  bone1Name=calf_r
  bone2Name=thigh_r
  swing1LimitDegrees=45
  swing1Motion=Free
  swing2LimitDegrees=45
  swing2Motion=Free
  twistLimitDegrees=45
  twistMotion=Free
 }
 {
  bone1Name=calf_l
  bone2Name=thigh_l
  swing1LimitDegrees=45
  swing1Motion=Free
  swing2LimitDegrees=45
  swing2Motion=Free
  twistLimitDegrees=45
  twistMotion=Free
 }
 {
  bone1Name=pelvis
  bone2Name=root
  swing1LimitDegrees=45
  swing1Motion=Free
  swing2LimitDegrees=45
  swing2Motion=Free
  twistLimitDegrees=45
  twistMotion=Free
 }
]
